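# Physics asset for the belt mesh of the female medium Typhoon character body, set up for a right-side knife and its strap.
# A limited-angle constraint chain links the knife master bone to the knife and two strap segments, with tight knife limits and looser strap swing.
# Capsules on pelvis, spine and right thigh twist give body shapes, joined by free-rotation constraints, as is the knife master to the pelvis.
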PhysicsAsset /BRCosmetics/Characters/Player/Female/Medium/Bodies/F_MED_Typhoon/Meshes/F_MED_Typhoon_Belt_Physics
bodies {
 dyn_knife_01_r [{shapeType=Capsule, length=30.544, radius=3.3487, center=<0.0001 2.534 0>, rotation=<rot 0.0001 0 -89.9994>}]
 dyn_knife_01_r_MSTR [{shapeType=Capsule, length=0.01, radius=0.01}]
 dyn_knifestrap_01_r [{shapeType=Capsule, length=3.51, radius=1.75, center=<3.5852 0.0001 0>, rotation=<rot -90 -14.8511 14.8512>}]
 dyn_knifestrap_02_r [{shapeType=Capsule, length=3.51, radius=1.75, center=<3.5852 0.0001 0>, rotation=<rot -90 -14.8511 14.8512>}]
 pelvis [{shapeType=Capsule, length=15.5452, radius=6.9577, center=<0 1.6062 0>}]
 spine_01 [{shapeType=Capsule, length=15.5452, radius=6.9577, center=<0 3.3947 0>}]
 spine_03 [{shapeType=Capsule, length=0.505, radius=10, shapeName=spine_03_capsule, center=<0 1.0849 0>}]
 thigh_twist_01_r [{shapeType=Capsule, length=18.5611, radius=6.6423, center=<-14.0084 0.0001 0>, rotation=<rot 90 -0.0198 -0.0198>}]
}
constraints [
 {bone1Name=dyn_knife_01_r, bone2Name=dyn_knife_01_r_MSTR, swing1LimitDegrees=5, swing2LimitDegrees=5, twistLimitDegrees=5}
 {bone1Name=dyn_knifestrap_01_r, bone2Name=dyn_knife_01_r, swing1LimitDegrees=45, swing2LimitDegrees=45, twistLimitDegrees=5}
 {bone1Name=dyn_knifestrap_02_r, bone2Name=dyn_knifestrap_01_r, swing1LimitDegrees=45, swing2LimitDegrees=45, twistLimitDegrees=5}
 {
  bone1Name=dyn_knife_01_r_MSTR
  bone2Name=pelvis
  swing1LimitDegrees=45
  swing1Motion=Free
  swing2LimitDegrees=45
  swing2Motion=Free
  twistLimitDegrees=45
  twistMotion=Free
 }
 {
  bone1Name=spine_01
  bone2Name=pelvis
  swing1LimitDegrees=45
  swing1Motion=Free
  swing2LimitDegrees=45
  swing2Motion=Free
  twistLimitDegrees=45
  twistMotion=Free
 }
 {
  bone1Name=thigh_twist_01_r
  bone2Name=pelvis
  swing1LimitDegrees=45
  swing1Motion=Free
  swing2LimitDegrees=45
  swing2Motion=Free
  twistLimitDegrees=45
  twistMotion=Free
 }
 {
  bone1Name=spine_03
  bone2Name=spine_01
  swing1LimitDegrees=45
  swing1Motion=Free
  swing2LimitDegrees=45
  swing2Motion=Free
  twistLimitDegrees=45
  twistMotion=Free
 }
]
