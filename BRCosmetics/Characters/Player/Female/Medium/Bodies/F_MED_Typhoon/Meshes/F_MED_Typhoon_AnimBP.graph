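# Animation Blueprint for the F_MED_Typhoon female body cosmetic, adding secondary motion to its belt and right-side knife.
# The copied parent pose passes through a weighted spine/pelvis constraint, pelvis-space RigidBody belt physics, and a calf look-at.
# An additive knife offset scales with right-thigh rotation relative to the pelvis.
# Weapon, customization and skydiving events toggle the "Hide Element 01" material parameter, setting it while a Typhoon-tagged pickaxe is equipped outside skydiving.
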AnimBlueprint /BRCosmetics/Characters/Player/Female/Medium/Bodies/F_MED_Typhoon/Meshes/F_MED_Typhoon_AnimBP
target_skeleton=&/F_MED_Typhoon_Skeleton
graphs [
 AnimGraph {nodes [
   n1 = "Constraint - dyn_knife_01_r_MSTR (PPPP)"!(
    ComponentPose=Local To Component(LocalPose=Copy Pose From Mesh() {bUseAttachedParent=true})
    Alpha=1
    ConstraintWeights_0=0.1163
    ConstraintWeights_1:float=0.2006
    ConstraintWeights_2:float=0.0401
    ConstraintWeights_3:float=0.643
   ) -> (Pose) {
    alphaBoolBlend {blendOption=Linear}
    boneToModify {boneName=dyn_knife_01_r_MSTR}
    constraintSetup [{targetBone {boneName=spine_01}}, {targetBone {boneName=spine_02}}, {targetBone {boneName=spine_04}}, {targetBone {boneName=pelvis}}]
    constraintWeights [1,1,1,1]
   }
   n2 = RigidBody!(ComponentPose=@n1, Alpha, OverrideWorldGravity:Vector="0,0,-980", ExternalForce) -> (Pose) {
    bClampLinearTranslationLimitToRefPose=true
    bOverrideWorldGravity=true
    overridePhysicsAsset=&/F_MED_Typhoon_Belt_Physics
    simulationSpace=BaseBoneSpace
    alphaBoolBlend {blendOption=Linear}
    baseBoneRef {boneName=pelvis}
    componentAppliedLinearAccClamp=<1500 1500 1500>
    componentLinearAccScale=<0.3 0.3 0.45>
    componentLinearVelScale=<0.3 0.3 0.45>
    overrideWorldGravity {z=-980}
   }
   n3 = Look At - Bone: dyn_knife_01_r(ComponentPose=@n2, Alpha=0.9) {
    bUseLookUpAxis=true
    alphaBoolBlend {blendOption=Linear}
    boneToModify {boneName=dyn_knife_01_r}
    lookAtLocation {x=0}
    lookAtTarget {boneReference {boneName=calf_r}}
    lookAt_Axis {axis {y=0, z=-1}}
    lookUp_Axis {axis=<0 1>}
   }
   n4 = "Transform (Modify) Bone - Bone: dyn_knife_01_r_MSTR"(
    ComponentPose=@n3
    Alpha=Get Right Leg() -> (Right Leg:double)
    Translation="1.401,-0.137,2.200"
    Rotation="-27.241,19.788,-20.846"
    Scale="1,1,1"
   ) {
    rotationMode=BMM_Additive
    rotationSpace=BCS_BoneSpace
    translationMode=BMM_Additive
    translationSpace=BCS_BoneSpace
    alphaBoolBlend {blendOption=Linear}
    boneToModify {boneName=dyn_knife_01_r_MSTR}
    rotation=<rot -27.241 19.788 -20.846>
    translation=<1.401 -0.137 2.2>
   }
   Output Pose(Result=Component To Local(ComponentPose=@n4)) {layerGroup=None}
  ]}
 EventGraph {nodes [
   Event BlueprintUpdateAnimation()
   > Set Right Leg(Right Leg:double=GetAlphaFromReferenceTransformDelta(SkeletalMeshComponent=Get Skel Mesh(), BoneName=thigh_r, BaseName=pelvis, InputRange=(X=-19,Y=-100), TwistAxis="1,0,0")) -> (Output_Get:double)
   n1 = Event OnWeaponChanged()
   n2 = Get Dynamic Material Instance(execute=[@n1.then, @n10.then], SkeletalMeshComponent=GetPartSkeletalMeshComponent())
   > n3 = Sequence()
   n4 = PickaxeItemDefinitionHasAnyMetaTag(Tags=Make Array("[0]"=(TagName="Cosmetics.MetaTags.TyphoonFemale")), bRequiresEquipped=true)
   > n5 = Branch(Condition=@n4)
   > n6 = Branch(Condition=Get bIsSkyDiving())
   SetScalarParameterValue(execute=[@n8.then, @n6.then, @n7.else, @n5.else], self=@n2, ParameterName=Hide Element 01, Value=0)
   SetScalarParameterValue(execute=[@n6.else, @n7.then], self=@n2, ParameterName=Hide Element 01, Value=1)
   n7 = Branch(execute=[@n9.then, @n3.then_1], Condition=@n4)
   n8 = Event OnBeginSkydiving()
   n9 = Event OnEndSkydiving()
   Event BlueprintInitializeAnimation()
   > Set Skel Mesh(Skel Mesh=GetOwnerSkeletalMeshComponent())
   n10 = Event OnCharacterCustomizationLoaded()
  ]}
]
variables [
 Right Leg {type=number}
 Skel Mesh {type=SkeletalMeshComponent}
]
