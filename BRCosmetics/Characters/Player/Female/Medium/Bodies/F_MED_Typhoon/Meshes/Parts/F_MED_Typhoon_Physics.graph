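# Physics asset for the F_MED_Typhoon female player body, set up for secondary motion on the front hair and ponytail.
# Capsules on head, upper spine, clavicles and upper arms, plus two backpack shapes, form the body volumes, linked by free-angular constraints.
# Tiny MSTR anchor bodies join the head with free rotation, each rooting a hair strand or the three-link ponytail chain through limited swing and tight twist constraints.
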PhysicsAsset /BRCosmetics/Characters/Player/Female/Medium/Bodies/F_MED_Typhoon/Meshes/Parts/F_MED_Typhoon_Physics
bodies {
 clavicle_l [{shapeType=Capsule, length=5.76, radius=4.5, center=<4.2361 -0.0001 0>, rotation=<rot -90 2.5305 -2.5304>}]
 clavicle_r [{shapeType=Capsule, length=5.76, radius=4.5, center=<-4.2361 0.0001 0>, rotation=<rot -90 -178.7341 -1.2658>}]
 dyn_backpack_COL1 [{shapeType=Capsule, length=51.005, radius=5.755, center=<-0.0002 3.1208 6.6926>, rotation=<rot 24.9998 -90 0>}]
 dyn_backpack_COL2 [{shapeType=Sphere, radius=15.5}]
 dyn_hairfrontA_01_l [{shapeType=Capsule, length=2.3367, radius=1.4155, center=<0.2858 -0.0002 -0.0001>, rotation=<rot -90 0.0198 -0.0198>}]
 dyn_hairfrontA_01_l_MSTR [{shapeType=Capsule, length=0.01, radius=0.01}]
 dyn_hairfrontA_01_r [{shapeType=Capsule, length=2.3367, radius=1.4155, center=<0.2858 -0.0002 -0.0001>, rotation=<rot -90 0.0198 -0.0198>}]
 dyn_hairfrontA_01_r_MSTR [{shapeType=Capsule, length=0.01, radius=0.01}]
 dyn_hairfrontB_01_l [{shapeType=Capsule, length=2.3367, radius=1.4155, center=<0.2858 -0.0002 -0.0001>, rotation=<rot -90 0.0198 -0.0198>}]
 dyn_hairfrontB_01_l_MSTR [{shapeType=Capsule, length=0.01, radius=0.01}]
 dyn_hairfrontB_01_r [{shapeType=Capsule, length=2.3367, radius=1.4155, center=<0.2858 -0.0002 -0.0001>, rotation=<rot -90 0.0198 -0.0198>}]
 dyn_hairfrontB_01_r_MSTR [{shapeType=Capsule, length=0.01, radius=0.01}]
 dyn_hairfront_01 [{shapeType=Capsule, length=2.3367, radius=1.4155, center=<0.2858 -0.0002 -0.0001>, rotation=<rot -90 0.0198 -0.0198>}]
 dyn_hairfront_01_MSTR [{shapeType=Capsule, length=0.01, radius=0.01}]
 dyn_ponytail_01 [{shapeType=Capsule, length=3.5867, radius=3.9155, center=<5.521 -0.0002 0>, rotation=<rot -90 -0.0264 0.0264>}]
 dyn_ponytail_01_MSTR [{shapeType=Capsule, length=0.01, radius=0.01}]
 dyn_ponytail_02 [{shapeType=Capsule, length=6.8367, radius=4.1655, center=<6.0262 -0.0002 0>, rotation=<rot -90 -0.0264 0.0264>}]
 dyn_ponytail_03 [{shapeType=Capsule, length=4.0867, radius=3.4155, center=<5.521 -0.0002 0>, rotation=<rot -90 -0.0264 0.0264>}]
 head [{shapeType=Capsule, length=5.1531, radius=7.7501, center=<2.2284 1.0803 0>, rotation=<rot -90 -0.0198 0.0198>}]
 spine_05 [{shapeType=Capsule, length=22.4031, radius=8.5001, center=<0 1.0212 0>}]
 upperarm_l [{shapeType=Capsule, length=19.01, radius=4.5, center=<12.5982 -0.0002 0>, rotation=<rot -90 -3.7927 3.7927>}]
 upperarm_r [{shapeType=Capsule, length=19.01, radius=4.5, center=<-12.5982 0.0002 0>, rotation=<rot -90 5.0511 174.9489>}]
}
constraints [
 {
  bone1Name=dyn_hairfrontA_01_l_MSTR
  bone2Name=head
  swing1LimitDegrees=45
  swing1Motion=Free
  swing2LimitDegrees=45
  swing2Motion=Free
  twistLimitDegrees=45
  twistMotion=Free
 }
 {
  bone1Name=dyn_hairfrontB_01_l_MSTR
  bone2Name=head
  swing1LimitDegrees=45
  swing1Motion=Free
  swing2LimitDegrees=45
  swing2Motion=Free
  twistLimitDegrees=45
  twistMotion=Free
 }
 {
  bone1Name=dyn_hairfront_01_MSTR
  bone2Name=head
  swing1LimitDegrees=45
  swing1Motion=Free
  swing2LimitDegrees=45
  swing2Motion=Free
  twistLimitDegrees=45
  twistMotion=Free
 }
 {
  bone1Name=dyn_hairfrontB_01_r_MSTR
  bone2Name=head
  swing1LimitDegrees=45
  swing1Motion=Free
  swing2LimitDegrees=45
  swing2Motion=Free
  twistLimitDegrees=45
  twistMotion=Free
 }
 {
  bone1Name=dyn_hairfrontA_01_r_MSTR
  bone2Name=head
  swing1LimitDegrees=45
  swing1Motion=Free
  swing2LimitDegrees=45
  swing2Motion=Free
  twistLimitDegrees=45
  twistMotion=Free
 }
 {
  bone1Name=dyn_ponytail_01_MSTR
  bone2Name=head
  swing1LimitDegrees=45
  swing1Motion=Free
  swing2LimitDegrees=45
  swing2Motion=Free
  twistLimitDegrees=45
  twistMotion=Free
 }
 {bone1Name=dyn_ponytail_01, bone2Name=dyn_ponytail_01_MSTR, swing1LimitDegrees=30, swing2LimitDegrees=30, twistLimitDegrees=5}
 {bone1Name=dyn_ponytail_02, bone2Name=dyn_ponytail_01, swing1LimitDegrees=30, swing2LimitDegrees=30, twistLimitDegrees=5}
 {bone1Name=dyn_ponytail_03, bone2Name=dyn_ponytail_02, swing1LimitDegrees=30, swing2LimitDegrees=30, twistLimitDegrees=5}
 {
  bone1Name=head
  bone2Name=spine_05
  swing1LimitDegrees=45
  swing1Motion=Free
  swing2LimitDegrees=45
  swing2Motion=Free
  twistLimitDegrees=45
  twistMotion=Free
 }
 {
  bone1Name=clavicle_l
  bone2Name=spine_05
  swing1LimitDegrees=45
  swing1Motion=Free
  swing2LimitDegrees=45
  swing2Motion=Free
  twistLimitDegrees=45
  twistMotion=Free
 }
 {
  bone1Name=clavicle_r
  bone2Name=spine_05
  swing1LimitDegrees=45
  swing1Motion=Free
  swing2LimitDegrees=45
  swing2Motion=Free
  twistLimitDegrees=45
  twistMotion=Free
 }
 {
  bone1Name=upperarm_l
  bone2Name=clavicle_l
  swing1LimitDegrees=45
  swing1Motion=Free
  swing2LimitDegrees=45
  swing2Motion=Free
  twistLimitDegrees=45
  twistMotion=Free
 }
 {
  bone1Name=upperarm_r
  bone2Name=clavicle_r
  swing1LimitDegrees=45
  swing1Motion=Free
  swing2LimitDegrees=45
  swing2Motion=Free
  twistLimitDegrees=45
  twistMotion=Free
 }
 {bone1Name=dyn_hairfrontA_01_l, bone2Name=dyn_hairfrontA_01_l_MSTR, swing1LimitDegrees=20, swing2LimitDegrees=20, twistLimitDegrees=5}
 {bone1Name=dyn_hairfrontB_01_l, bone2Name=dyn_hairfrontB_01_l_MSTR, swing1LimitDegrees=20, swing2LimitDegrees=20, twistLimitDegrees=5}
 {bone1Name=dyn_hairfront_01, bone2Name=dyn_hairfront_01_MSTR, swing1LimitDegrees=20, swing2LimitDegrees=20, twistLimitDegrees=5}
 {bone1Name=dyn_hairfrontB_01_r, bone2Name=dyn_hairfrontB_01_r_MSTR, swing1LimitDegrees=20, swing2LimitDegrees=20, twistLimitDegrees=5}
 {bone1Name=dyn_hairfrontA_01_r, bone2Name=dyn_hairfrontA_01_r_MSTR, swing1LimitDegrees=20, swing2LimitDegrees=20, twistLimitDegrees=5}
 {
  bone1Name=dyn_backpack_COL1
  bone2Name=spine_05
  swing1LimitDegrees=45
  swing1Motion=Free
  swing2LimitDegrees=45
  swing2Motion=Free
  twistLimitDegrees=45
  twistMotion=Free
 }
 {
  bone1Name=dyn_backpack_COL2
  bone2Name=spine_05
  swing1LimitDegrees=45
  swing1Motion=Free
  swing2LimitDegrees=45
  swing2Motion=Free
  twistLimitDegrees=45
  twistMotion=Free
 }
]
