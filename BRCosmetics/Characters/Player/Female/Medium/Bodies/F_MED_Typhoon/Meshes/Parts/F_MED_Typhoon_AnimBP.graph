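# Post-process style AnimBP for Typhoon's body part that copies the parent pose and simulates head-relative RigidBody dynamics, including her ponytail.
# Backpack identity and type checks offset two collision bones before simulation.
# Mid-height backpack bounds drive ponytail constraints to spine_05 afterwards.
# Each update zeroes gravity while skydiving and sets the misleadingly named isRbanEnabled when RigidBody is disabled or LOD exceeds 1, constraining dyn_ponytail_02.
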AnimBlueprint /BRCosmetics/Characters/Player/Female/Medium/Bodies/F_MED_Typhoon/Meshes/Parts/F_MED_Typhoon_AnimBP
target_skeleton=&/F_MED_Typhoon_Skeleton
graphs [
 AnimGraph {nodes [
   n1 = "Transform (Modify) Bone - Bone: dyn_backpack_COL1"(
    ComponentPose=Local To Component(LocalPose=Copy Pose From Mesh() {bUseAttachedParent=true})
    Alpha=Get isWearingTyphoonBackpack() -> (isWearingTyphoonBackpack:double)
    Translation="41.359,5.835,-0.000"
    Rotation="0,0,0"
    Scale="1,1,1"
   ) {
    rotationMode=BMM_Additive
    rotationSpace=BCS_BoneSpace
    scaleMode=BMM_Additive
    scaleSpace=BCS_BoneSpace
    translationMode=BMM_Additive
    translationSpace=BCS_BoneSpace
    alphaBoolBlend {blendOption=Linear}
    boneToModify {boneName=dyn_backpack_COL1}
    translation=<41.359 5.835>
   }
   n2 = "Transform (Modify) Bone - Bone: dyn_backpack_COL2"(
    ComponentPose=@n1
    Alpha=Get isWearingBackbling() -> (isWearingBackbling:double)
    Translation="37.195,0.000,-6.262"
    Rotation="0,0,0"
    Scale="1,1,1"
   ) {
    rotationMode=BMM_Additive
    rotationSpace=BCS_BoneSpace
    scaleMode=BMM_Additive
    scaleSpace=BCS_BoneSpace
    translationMode=BMM_Additive
    translationSpace=BCS_BoneSpace
    alphaBoolBlend {blendOption=Linear}
    boneToModify {boneName=dyn_backpack_COL2}
    translation {x=37.195, z=-6.262}
   }
   n3 = RigidBody!(ComponentPose=@n2, Alpha, OverrideWorldGravity:Vector=Get gravityVector() -> (gravityVector:Vector), ExternalForce) -> (Pose) {
    bClampLinearTranslationLimitToRefPose=true
    bOverrideWorldGravity=true
    overridePhysicsAsset=&/F_MED_Typhoon_Physics
    simulationSpace=BaseBoneSpace
    alphaBoolBlend {blendOption=Linear}
    baseBoneRef {boneName=head}
    componentAppliedLinearAccClamp=<1500 1500 1500>
    componentLinearAccScale=<0.3 0.3 0.45>
    componentLinearVelScale=<0.3 0.3 0.45>
    overrideWorldGravity {z=-980}
   }
   n4 = Get isLargePackAlpha() -> (isLargePackAlpha:double)
   n5 = "Constraint - dyn_ponytail_02 (P)"(ComponentPose=@n3, Alpha=@n4, ConstraintWeights_0=1) {
    alphaBoolBlend {blendOption=Linear}
    alphaScaleBiasClamp {bInterpResult=true}
    boneToModify {boneName=dyn_ponytail_02}
    constraintSetup [{targetBone {boneName=spine_05}}]
    constraintWeights [1]
   }
   n6 = "Constraint - dyn_ponytail_03 (P)"(ComponentPose=@n5, Alpha=@n4, ConstraintWeights_0=1) {
    alphaBoolBlend {blendOption=Linear}
    alphaScaleBiasClamp {bInterpResult=true}
    boneToModify {boneName=dyn_ponytail_03}
    constraintSetup [{targetBone {boneName=spine_05}}]
    constraintWeights [1]
   }
   n7 = "Constraint - dyn_ponytail_02 (P)"(ComponentPose=@n6, Alpha=Get isRbanEnabled() -> (isRbanEnabled:double), ConstraintWeights_0=1) {
    alphaBoolBlend {blendOption=Linear}
    alphaScaleBiasClamp {bInterpResult=true}
    boneToModify {boneName=dyn_ponytail_02}
    constraintSetup [{targetBone {boneName=spine_05}}]
    constraintWeights [1]
   }
   Output Pose(Result=Component To Local(ComponentPose=@n7)) {layerGroup=None}
  ]}
 EventGraph {nodes [
   Event OnCharacterCustomizationLoaded()
   > n1 = Cast To FortPlayerPawn(Object=GetOwningActor())
   > Is Valid(InputObject=Get typhoonBackblingSkelMesh() -> (typhoonBackblingSkelMesh:Skeletal Mesh Object Reference))
   > n2 = Set currentBackpack(currentBackpack:Athena Backpack Item Definition Object Reference=GetItemDefinitionTypeFromLoadoutWithActor!(
     Actor:Actor Object Reference=@n1
     LoadoutSlot:Gameplay Tag Structure=(TagName="LoadoutSlot.Backpack")
     CosmeticType:Athena Cosmetic Item Definition Class Reference=/Script/FortniteGame.AthenaBackpackItemDefinition
    ) -> (OutCosmeticItemDef:Athena Backpack Item Definition Object Reference)) -> (Output_Get:Athena Backpack Item Definition Object Reference)
   > n3 = Set isTyphoonBackbling(isTyphoonBackbling:bool="Equal (Object)"(A=@n2, B=Get backblingItemDef() -> (backblingItemDef:Athena Backpack Item Definition Object Reference)) {is=title K2Node_CallFunction}) -> (Output_Get:bool)
   > Set isWearingTyphoonBackpack(isWearingTyphoonBackpack:double="To Float (Boolean)"(InBool=@n3)) -> (Output_Get:double)
   > Do Once()
   n4 = NOT Boolean(A=Get isTyphoonBackbling() -> (isTyphoonBackbling:bool))
   n5 = CheckForBackbling()
   n6 = NOT Boolean(A=@n5.bIsBackblingLong)
   > n7 = Set isWearingBackbling(isWearingBackbling:double="To Float (Boolean)"(InBool=AND Boolean!(A=@n5, B=@n6, C:bool=@n4) -> (ReturnValue))) -> (Output_Get:double=0)
   n8 = Get playerPawn() -> (playerPawn:Fort Player Pawn Athena Object Reference)
   Is Valid(exec=[@n14.then_0, @n7.then], InputObject=@n8)
   n9 = GetComponentBounds!(Component:Scene Component Object Reference=GetSkeletalMeshForPartType(self=@n8, PartType=Backpack)) -> (Origin:Vector, BoxExtent:Vector, SphereRadius:float)
   n10 = BreakVector(InVec=@n9.BoxExtent)
   n11 = "float > float"(A=@n10.Z, B=15) {is=title K2Node_CallFunction}
   n12 = "float <= float"!(A:double=@n10.Z, B:double=50) -> (ReturnValue:bool) {is=title K2Node_CallFunction}
   n13 = NOT Boolean(A=Get isTyphoonBackbling())
   > Set isLargePackAlpha(isLargePackAlpha:double="To Float (Boolean)"(InBool=AND Boolean(A=@n11, B=@n12, C=@n13))) -> (Output_Get:double)
   Event BlueprintUpdateAnimation()
   > n14 = Sequence!(execute) -> (then_0, then_1, then_2:Exec)
   Set gravityVector(execute=@n14.then_1, gravityVector:Vector=SelectVector(A="0, 0, 0", B="0, 0,-980", bPickA=Get bIsSkyDiving())) -> (Output_Get:Vector)
   n15 = "integer > integer"(A=Get CurrentLOD(), B=1) {is=title K2Node_CallFunction}
   n16 = NOT Boolean(A=Get bRigidBodyEnabled())
   Set isRbanEnabled(execute=@n14.then_2, isRbanEnabled:double="To Float (Boolean)"(InBool=OR Boolean(A=@n16, B=@n15))) -> (Output_Get:double)
   Event BlueprintInitializeAnimation()
   n17 = TryGetPawnOwner()
   > Is Valid(InputObject=@n17)
   > n18 = Cast To FortPlayerPawnAthena(Object=@n17)
   > Set playerPawn(playerPawn:Fort Player Pawn Athena Object Reference=@n18) -> (Output_Get:Fort Player Pawn Athena Object Reference)
   DrawDebugBox!(
    execute
    Center:Vector=@n9
    Extent:Vector=@n9.BoxExtent
    LineColor:Linear Color Structure
    Rotation:Rotator
    Duration:float=0
    Thickness:float=5
    DepthPriority:EDrawDebugSceneDepthPriorityGroup Enum=World
   ) -> (then)
  ]}
]
variables [
 backblingItemDef {type=AthenaBackpackItemDefinition, default=/BRCosmetics/Athena/Items/Cosmetics/Backpacks/BID_690_TyphoonFemale_1NBFZ}
 currentBackpack {type=AthenaBackpackItemDefinition, default=null}
 gravityVector {type=Vector}
 isLargePackAlpha {type=number}
 isRbanEnabled {type=number}
 isSmallPackAlpha {type=number}
 isTyphoonBackbling {type=boolean}
 isWearingBackbling {type=number}
 isWearingCape {type=number}
 isWearingTyphoonBackpack {type=number}
 NewVar_0 {
  type=FortAthenaLoadout
  default {
   aura=null
   backpack=null
   bannerColorId=""
   bannerIconId=""
   bass=null
   character=null
   drums=null
   glider=null
   guitar=null
   keytar=null
   loadingScreen=null
   mic=null
   mimosa=null
   musicPack=null
   pickaxe=null
   shoes=null
   skyDiveContrail=null
   characterVariantChannels []
   dances []
   itemWraps []
  }
 }
 NewVar_1 {type=SizeBox, default=null}
 NewVar_2 {type=SkeletalMesh, default=null}
 NewVar_3 {type=Class, default=null}
 NewVar_4 {type=Class, default=null}
 playerPawn {type=FortPlayerPawnAthena}
 test {type=string, default=""}
 typhoonBackblingSkelMesh {type=SkeletalMesh, default=/BRCosmetics/Accessories/FORT_Backpacks/Backpack_F_MED_Typhoon/Meshes/F_MED_Typhoon_Pack}
]
